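# Physics asset for secondary motion of dangling gear on the medium male BulletBlue player body cosmetic: belt straps, hand straps, grenades and pocket flaps.
# Each dynamic bone has a capsule or box body, constrained with limited swing and twist to its own small _MSTR capsule.
# Every _MSTR body links to a single spine_02 capsule with free angular motion, leaving the sway limits to the child constraints.
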PhysicsAsset /BRCosmetics/Characters/Player/Male/Medium/Bodies/M_MED_BulletBlue/Meshes/M_MED_Bullet_Blue_Dyn
bodies {
 dyn_belt_strap_l_01 [{shapeType=Capsule, length=2, radius=1, center=<1 0 0>, rotation=<rot 90 0 0>}]
 dyn_belt_strap_l_01_MSTR [{shapeType=Capsule, length=1, radius=1}]
 dyn_belt_strap_l_02 [{shapeType=Capsule, length=2, radius=1, center=<1.5 0 0>, rotation=<rot 90 0 0>}]
 dyn_belt_strap_l_02_MSTR [{shapeType=Capsule, length=1, radius=1}]
 dyn_belt_strap_r_01 [{shapeType=Capsule, length=2, radius=1, center=<2 0 0>, rotation=<rot 90 0 0>}]
 dyn_belt_strap_r_01_MSTR [{shapeType=Capsule, length=1, radius=1}]
 dyn_belt_strap_r_02 [{shapeType=Capsule, length=2, radius=1, center=<1 0 0>, rotation=<rot 90 0 0>}]
 dyn_belt_strap_r_02_MSTR [{shapeType=Capsule, length=1, radius=1}]
 dyn_grenade_l [{shapeType=Capsule, length=9, radius=4, center=<-2 0 1.5>, rotation=<rot 90 0 0>}]
 dyn_grenade_l_MSTR [{shapeType=Capsule, length=1, radius=1, rotation=<rot 90 0 0>}]
 dyn_grenade_r [{shapeType=Capsule, length=9, radius=4, center=<-2 0 1>, rotation=<rot 90 0 0>}]
 dyn_grenade_r_MSTR [{shapeType=Capsule, length=1, radius=1, rotation=<rot 90 0 0>}]
 dyn_hand_strap_l_01 [{shapeType=Capsule, length=2, radius=1, center=<1.5 0 0>, rotation=<rot 90 0 0>}]
 dyn_hand_strap_l_01_MSTR [{shapeType=Capsule, length=1, radius=1}]
 dyn_hand_strap_r_01 [{shapeType=Capsule, length=2, radius=1, center=<1 0 0>, rotation=<rot 90 0 0>}]
 dyn_hand_strap_r_01_MSTR [{shapeType=Capsule, length=1, radius=1}]
 dyn_pocketFlap_01_l [{shapeType=Box, extentX=15, extentY=10, extentZ=2, center=<7 0 0>}]
 dyn_pocketFlap_01_l_MSTR [{shapeType=Capsule, length=1, radius=1}]
 dyn_pocketFlap_01_r [{shapeType=Box, extentX=15, extentY=10, extentZ=2, center=<7 0 0>}]
 dyn_pocketFlap_01_r_MSTR [{shapeType=Capsule, length=1, radius=1}]
 dyn_pocketFlap_02_l [{shapeType=Box, extentX=15, extentY=10, extentZ=2, center=<7 0 0>}]
 dyn_pocketFlap_02_l_MSTR [{shapeType=Capsule, length=1, radius=1}]
 dyn_pocketFlap_02_r [{shapeType=Box, extentX=15, extentY=10, extentZ=2, center=<7 0 0>}]
 dyn_pocketFlap_02_r_MSTR [{shapeType=Capsule, length=1, radius=1}]
 spine_02 [{shapeType=Capsule, length=3.2093, radius=16.5675, center=<0.0001 -1.3064 0>}]
}
constraints [
 {bone1Name=dyn_grenade_l, bone2Name=dyn_grenade_l_MSTR, swing1LimitDegrees=10, swing2LimitDegrees=5, twistLimitDegrees=15}
 {bone1Name=dyn_grenade_r, bone2Name=dyn_grenade_r_MSTR, swing1LimitDegrees=10, swing2LimitDegrees=5, twistLimitDegrees=15}
 {bone1Name=dyn_pocketFlap_02_l, bone2Name=dyn_pocketFlap_02_l_MSTR, swing1LimitDegrees=15, swing2LimitDegrees=45, twistLimitDegrees=2}
 {bone1Name=dyn_pocketFlap_01_l, bone2Name=dyn_pocketFlap_01_l_MSTR, swing1LimitDegrees=15, swing2LimitDegrees=45, twistLimitDegrees=2}
 {bone1Name=dyn_pocketFlap_02_r, bone2Name=dyn_pocketFlap_02_r_MSTR, swing1LimitDegrees=15, swing2LimitDegrees=45, twistLimitDegrees=2}
 {bone1Name=dyn_pocketFlap_01_r, bone2Name=dyn_pocketFlap_01_r_MSTR, swing1LimitDegrees=15, swing2LimitDegrees=45, twistLimitDegrees=2}
 {
  bone1Name=dyn_grenade_l_MSTR
  bone2Name=spine_02
  swing1LimitDegrees=45
  swing1Motion=Free
  swing2LimitDegrees=45
  swing2Motion=Free
  twistLimitDegrees=45
  twistMotion=Free
 }
 {
  bone1Name=dyn_grenade_r_MSTR
  bone2Name=spine_02
  swing1LimitDegrees=45
  swing1Motion=Free
  swing2LimitDegrees=45
  swing2Motion=Free
  twistLimitDegrees=45
  twistMotion=Free
 }
 {
  bone1Name=dyn_pocketFlap_02_l_MSTR
  bone2Name=spine_02
  swing1LimitDegrees=45
  swing1Motion=Free
  swing2LimitDegrees=45
  swing2Motion=Free
  twistLimitDegrees=45
  twistMotion=Free
 }
 {
  bone1Name=dyn_pocketFlap_01_l_MSTR
  bone2Name=spine_02
  swing1LimitDegrees=45
  swing1Motion=Free
  swing2LimitDegrees=45
  swing2Motion=Free
  twistLimitDegrees=45
  twistMotion=Free
 }
 {
  bone1Name=dyn_pocketFlap_02_r_MSTR
  bone2Name=spine_02
  swing1LimitDegrees=45
  swing1Motion=Free
  swing2LimitDegrees=45
  swing2Motion=Free
  twistLimitDegrees=45
  twistMotion=Free
 }
 {
  bone1Name=dyn_pocketFlap_01_r_MSTR
  bone2Name=spine_02
  swing1LimitDegrees=45
  swing1Motion=Free
  swing2LimitDegrees=45
  swing2Motion=Free
  twistLimitDegrees=45
  twistMotion=Free
 }
 {
  bone1Name=dyn_belt_strap_l_02_MSTR
  bone2Name=spine_02
  swing1LimitDegrees=45
  swing1Motion=Free
  swing2LimitDegrees=45
  swing2Motion=Free
  twistLimitDegrees=45
  twistMotion=Free
 }
 {bone1Name=dyn_belt_strap_l_02, bone2Name=dyn_belt_strap_l_02_MSTR, swing1LimitDegrees=30, swing2LimitDegrees=15, twistLimitDegrees=2}
 {
  bone1Name=dyn_belt_strap_r_02_MSTR
  bone2Name=spine_02
  swing1LimitDegrees=45
  swing1Motion=Free
  swing2LimitDegrees=45
  swing2Motion=Free
  twistLimitDegrees=45
  twistMotion=Free
 }
 {bone1Name=dyn_belt_strap_r_02, bone2Name=dyn_belt_strap_r_02_MSTR, swing1LimitDegrees=30, swing2LimitDegrees=15, twistLimitDegrees=2}
 {
  bone1Name=dyn_belt_strap_l_01_MSTR
  bone2Name=spine_02
  swing1LimitDegrees=45
  swing1Motion=Free
  swing2LimitDegrees=45
  swing2Motion=Free
  twistLimitDegrees=45
  twistMotion=Free
 }
 {bone1Name=dyn_belt_strap_l_01, bone2Name=dyn_belt_strap_l_01_MSTR, swing1LimitDegrees=15, swing2LimitDegrees=30, twistLimitDegrees=2}
 {
  bone1Name=dyn_belt_strap_r_01_MSTR
  bone2Name=spine_02
  swing1LimitDegrees=45
  swing1Motion=Free
  swing2LimitDegrees=45
  swing2Motion=Free
  twistLimitDegrees=45
  twistMotion=Free
 }
 {bone1Name=dyn_belt_strap_r_01, bone2Name=dyn_belt_strap_r_01_MSTR, swing1LimitDegrees=20, swing2LimitDegrees=30, twistLimitDegrees=2}
 {
  bone1Name=dyn_hand_strap_r_01_MSTR
  bone2Name=spine_02
  swing1LimitDegrees=45
  swing1Motion=Free
  swing2LimitDegrees=45
  swing2Motion=Free
  twistLimitDegrees=45
  twistMotion=Free
 }
 {bone1Name=dyn_hand_strap_r_01, bone2Name=dyn_hand_strap_r_01_MSTR, swing1LimitDegrees=15, swing2LimitDegrees=30, twistLimitDegrees=2}
 {
  bone1Name=dyn_hand_strap_l_01_MSTR
  bone2Name=spine_02
  swing1LimitDegrees=45
  swing1Motion=Free
  swing2LimitDegrees=45
  swing2Motion=Free
  twistLimitDegrees=45
  twistMotion=Free
 }
 {bone1Name=dyn_hand_strap_l_01, bone2Name=dyn_hand_strap_l_01_MSTR, swing1LimitDegrees=15, swing2LimitDegrees=30, twistLimitDegrees=2}
]
