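# Physics asset for the dynamic leg accessories of the Bullet Blue medium male player body: a pistol and a leg strap on each side.
# Boxes cover the pistols and straps, while tiny capsules on the pistol master bones and calf twist bones act as anchors.
# Mirrored constraints allow each pistol only a few degrees of motion against its master, while straps swing more widely against the calves with little twist.
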PhysicsAsset /BRCosmetics/Characters/Player/Male/Medium/Bodies/M_MED_BulletBlue/Meshes/M_MED_Bullet_Blue_Dyn_legs
bodies {
 calf_twist_02_l [{shapeType=Capsule, length=1, radius=1}]
 calf_twist_02_r [{shapeType=Capsule, length=1, radius=1}]
 dyn_legStrap_l [{shapeType=Box, extentX=6, extentY=4, extentZ=1.5, center=<2.5 0 0>}]
 dyn_legStrap_r [{shapeType=Box, extentX=6, extentY=4, extentZ=1.5, center=<1.5 0 0>}]
 dyn_pistol_l [{shapeType=Box, extentX=22, extentY=12, extentZ=7.5, center=<4 0 2>}]
 dyn_pistol_l_MSTR [{shapeType=Capsule, length=1, radius=1, rotation=<rot 90 0 0>}]
 dyn_pistol_r [{shapeType=Box, extentX=22, extentY=12, extentZ=7.5, center=<4 0 2>}]
 dyn_pistol_r_MSTR [{shapeType=Capsule, length=1, radius=1, rotation=<rot 90 0 0>}]
}
constraints [
 {bone1Name=dyn_pistol_l, bone2Name=dyn_pistol_l_MSTR, swing1LimitDegrees=3, swing2LimitDegrees=2, twistLimitDegrees=5}
 {bone1Name=dyn_pistol_r, bone2Name=dyn_pistol_r_MSTR, swing1LimitDegrees=3, swing2LimitDegrees=2, twistLimitDegrees=5}
 {bone1Name=dyn_legStrap_l, bone2Name=calf_twist_02_l, swing1LimitDegrees=15, swing2LimitDegrees=45, twistLimitDegrees=2}
 {bone1Name=dyn_legStrap_r, bone2Name=calf_twist_02_r, swing1LimitDegrees=15, swing2LimitDegrees=45, twistLimitDegrees=2}
]
