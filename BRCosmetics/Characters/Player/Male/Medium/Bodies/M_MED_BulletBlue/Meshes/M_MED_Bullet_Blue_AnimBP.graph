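# Post-pose AnimBlueprint for the Bullet Blue medium male cosmetic body, animating shells, grenades, pocket flaps, pistols and straps.
# It copies the attached parent's pose, positions accessory bones with weighted constraints, and orients shells and pistols with Look At.
# Additive nudges shift middle-shell aims when parachuting and the right grenade via an aim-down-sights alpha inferred from right upper-arm rotation.
# Two RigidBody passes, based at spine_05 and pelvis, add secondary physics.
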AnimBlueprint /BRCosmetics/Characters/Player/Male/Medium/Bodies/M_MED_BulletBlue/Meshes/M_MED_Bullet_Blue_AnimBP
target_skeleton=&/M_MED_Bullet_Blue_Skeleton
graphs [
 AnimGraph {nodes [
   n1 = "Constraint - dyn_shells_01_r_MSTR (PP)"!(
    ComponentPose=Local To Component(LocalPose=Copy Pose From Mesh() {bUseAttachedParent=true})
    Alpha=1
    ConstraintWeights_0=0.813
    ConstraintWeights_1:float=0.187
   ) -> (Pose) {
    alphaBoolBlend {blendOption=Linear}
    boneToModify {boneName=dyn_shells_01_r_MSTR}
    constraintSetup [{targetBone {boneName=spine_05}}, {targetBone {boneName=clavicle_r}}]
    constraintWeights [1,1]
   }
   n2 = "Constraint - dyn_shells_02_r_MSTR (PP)"!(ComponentPose=@n1, Alpha=1, ConstraintWeights_0=0.852, ConstraintWeights_1:float=0.148) -> (Pose) {
    alphaBoolBlend {blendOption=Linear}
    boneToModify {boneName=dyn_shells_02_r_MSTR}
    constraintSetup [{targetBone {boneName=spine_05}}, {targetBone {boneName=clavicle_r}}]
    constraintWeights [1,1]
   }
   n3 = "Constraint - dyn_shells_03_r_MSTR (PP)"!(ComponentPose=@n2, Alpha=1, ConstraintWeights_0=0.968, ConstraintWeights_1:float=0.032) -> (Pose) {
    alphaBoolBlend {blendOption=Linear}
    boneToModify {boneName=dyn_shells_03_r_MSTR}
    constraintSetup [{targetBone {boneName=spine_05}}, {targetBone {boneName=clavicle_r}}]
    constraintWeights [1,1]
   }
   n4 = "Constraint - dyn_shells_01_r_aim (PPP)"!(ComponentPose=@n3, Alpha=1, ConstraintWeights_0=0.562, ConstraintWeights_1:float=0.376, ConstraintWeights_2:float=0.062) -> (Pose) {
    alphaBoolBlend {blendOption=Linear}
    boneToModify {boneName=dyn_shells_01_r_aim}
    constraintSetup [{targetBone {boneName=spine_05}}, {targetBone {boneName=clavicle_r}}, {targetBone {boneName=upperarm_twist_01_r}}]
    constraintWeights [1,1,1]
   }
   n5 = "Constraint - dyn_shells_02_r_aim (PPP)"!(ComponentPose=@n4, Alpha=1, ConstraintWeights_0=0.533, ConstraintWeights_1:float=0.357, ConstraintWeights_2:float=0.11) -> (Pose) {
    alphaBoolBlend {blendOption=Linear}
    boneToModify {boneName=dyn_shells_02_r_aim}
    constraintSetup [{targetBone {boneName=spine_05}}, {targetBone {boneName=clavicle_r}}, {targetBone {boneName=upperarm_twist_01_r}}]
    constraintWeights [1,1,1]
   }
   n6 = "Constraint - dyn_shells_03_r_aim (PPP)"!(ComponentPose=@n5, Alpha=1, ConstraintWeights_0=0.573, ConstraintWeights_1:float=0.358, ConstraintWeights_2:float=0.069) -> (Pose) {
    alphaBoolBlend {blendOption=Linear}
    boneToModify {boneName=dyn_shells_03_r_aim}
    constraintSetup [{targetBone {boneName=spine_05}}, {targetBone {boneName=clavicle_r}}, {targetBone {boneName=upperarm_twist_01_r}}]
    constraintWeights [1,1,1]
   }
   n7 = Get shells fix alpha() -> (shells fix alpha:double)
   n8 = "Transform (Modify) Bone - Bone: dyn_shells_02_r_aim"(ComponentPose=@n6, Alpha=@n7, Translation="0.000,0,1.75", Rotation="0,0,0", Scale="1,1,1") {
    translationMode=BMM_Additive
    translationSpace=BCS_BoneSpace
    alphaBoolBlend {blendOption=Linear}
    boneToModify {boneName=dyn_shells_02_r_aim}
    translation {z=1.75}
   }
   n9 = Look At - Bone: dyn_shells_01_r(ComponentPose=@n8, Alpha=1) {
    alphaBoolBlend {blendOption=Linear}
    boneToModify {boneName=dyn_shells_01_r}
    lookAtLocation {x=0}
    lookAtTarget {boneReference {boneName=dyn_shells_01_r_aim}}
    lookAt_Axis {axis=<1 0>}
   }
   n10 = Look At - Bone: dyn_shells_02_r(ComponentPose=@n9, Alpha=1) {
    alphaBoolBlend {blendOption=Linear}
    boneToModify {boneName=dyn_shells_02_r}
    lookAtLocation {x=0}
    lookAtTarget {boneReference {boneName=dyn_shells_02_r_aim}}
    lookAt_Axis {axis=<1 0>}
   }
   n11 = Look At - Bone: dyn_shells_03_r(ComponentPose=@n10, Alpha=1) {
    alphaBoolBlend {blendOption=Linear}
    boneToModify {boneName=dyn_shells_03_r}
    lookAtLocation {x=0}
    lookAtTarget {boneReference {boneName=dyn_shells_03_r_aim}}
    lookAt_Axis {axis=<1 0>}
   }
   n12 = "Constraint - dyn_shells_01_l_MSTR (PP)"!(ComponentPose=@n11, Alpha=1, ConstraintWeights_0=0.813, ConstraintWeights_1:float=0.187) -> (Pose) {
    alphaBoolBlend {blendOption=Linear}
    boneToModify {boneName=dyn_shells_01_l_MSTR}
    constraintSetup [{targetBone {boneName=spine_05}}, {targetBone {boneName=clavicle_l}}]
    constraintWeights [1,1]
   }
   n13 = "Constraint - dyn_shells_02_l_MSTR (PP)"!(ComponentPose=@n12, Alpha=1, ConstraintWeights_0=0.852, ConstraintWeights_1:float=0.148) -> (Pose) {
    alphaBoolBlend {blendOption=Linear}
    boneToModify {boneName=dyn_shells_02_l_MSTR}
    constraintSetup [{targetBone {boneName=spine_05}}, {targetBone {boneName=clavicle_l}}]
    constraintWeights [1,1]
   }
   n14 = "Constraint - dyn_shells_03_l_MSTR (PP)"!(ComponentPose=@n13, Alpha=1, ConstraintWeights_0=0.968, ConstraintWeights_1:float=0.032) -> (Pose) {
    alphaBoolBlend {blendOption=Linear}
    boneToModify {boneName=dyn_shells_03_l_MSTR}
    constraintSetup [{targetBone {boneName=spine_05}}, {targetBone {boneName=clavicle_l}}]
    constraintWeights [1,1]
   }
   n15 = "Constraint - dyn_shells_01_l_aim (PPP)"!(ComponentPose=@n14, Alpha=1, ConstraintWeights_0=0.562, ConstraintWeights_1:float=0.376, ConstraintWeights_2:float=0.062) -> (Pose) {
    alphaBoolBlend {blendOption=Linear}
    boneToModify {boneName=dyn_shells_01_l_aim}
    constraintSetup [{targetBone {boneName=spine_05}}, {targetBone {boneName=clavicle_l}}, {targetBone {boneName=upperarm_twist_01_l}}]
    constraintWeights [1,1,1]
   }
   n16 = "Constraint - dyn_shells_02_l_aim (PPP)"!(ComponentPose=@n15, Alpha=1, ConstraintWeights_0=0.533, ConstraintWeights_1:float=0.357, ConstraintWeights_2:float=0.11) -> (Pose) {
    alphaBoolBlend {blendOption=Linear}
    boneToModify {boneName=dyn_shells_02_l_aim}
    constraintSetup [{targetBone {boneName=spine_05}}, {targetBone {boneName=clavicle_l}}, {targetBone {boneName=upperarm_twist_01_l}}]
    constraintWeights [1,1,1]
   }
   n17 = "Constraint - dyn_shells_03_l_aim (PPP)"!(ComponentPose=@n16, Alpha=1, ConstraintWeights_0=0.573, ConstraintWeights_1:float=0.358, ConstraintWeights_2:float=0.069) -> (Pose) {
    alphaBoolBlend {blendOption=Linear}
    boneToModify {boneName=dyn_shells_03_l_aim}
    constraintSetup [{targetBone {boneName=spine_05}}, {targetBone {boneName=clavicle_l}}, {targetBone {boneName=upperarm_twist_01_l}}]
    constraintWeights [1,1,1]
   }
   n18 = "Transform (Modify) Bone - Bone: dyn_shells_02_l_aim"(ComponentPose=@n17, Alpha=@n7, Translation="0.000,0,1.75", Rotation="0,0,0", Scale="1,1,1") {
    translationMode=BMM_Additive
    translationSpace=BCS_BoneSpace
    alphaBoolBlend {blendOption=Linear}
    boneToModify {boneName=dyn_shells_02_l_aim}
    translation {z=1.75}
   }
   n19 = Look At - Bone: dyn_shells_01_l(ComponentPose=@n18, Alpha=1) {
    alphaBoolBlend {blendOption=Linear}
    boneToModify {boneName=dyn_shells_01_l}
    lookAtLocation {x=0}
    lookAtTarget {boneReference {boneName=dyn_shells_01_l_aim}}
    lookAt_Axis {axis=<1 0>}
   }
   n20 = Look At - Bone: dyn_shells_02_l(ComponentPose=@n19, Alpha=1) {
    alphaBoolBlend {blendOption=Linear}
    boneToModify {boneName=dyn_shells_02_l}
    lookAtLocation {x=0}
    lookAtTarget {boneReference {boneName=dyn_shells_02_l_aim}}
    lookAt_Axis {axis=<1 0>}
   }
   n21 = Look At - Bone: dyn_shells_03_l(ComponentPose=@n20, Alpha=1) {
    alphaBoolBlend {blendOption=Linear}
    boneToModify {boneName=dyn_shells_03_l}
    lookAtLocation {x=0}
    lookAtTarget {boneReference {boneName=dyn_shells_03_l_aim}}
    lookAt_Axis {axis=<1 0>}
   }
   n22 = "Constraint - dyn_grenade_l_MSTR (PP)"!(ComponentPose=@n21, Alpha=1, ConstraintWeights_0=0.2, ConstraintWeights_1:float=0.8) -> (Pose) {
    alphaBoolBlend {blendOption=Linear}
    boneToModify {boneName=dyn_grenade_l_MSTR}
    constraintSetup [{targetBone {boneName=spine_05}}, {targetBone {boneName=spine_04}}]
    constraintWeights [1,1]
   }
   n23 = "Constraint - dyn_grenade_r_MSTR (PP)"!(ComponentPose=@n22, Alpha=1, ConstraintWeights_0=0.2, ConstraintWeights_1:float=0.8) -> (Pose) {
    alphaBoolBlend {blendOption=Linear}
    boneToModify {boneName=dyn_grenade_r_MSTR}
    constraintSetup [{targetBone {boneName=spine_05}}, {targetBone {boneName=spine_04}}]
    constraintWeights [1,1]
   }
   n24 = "Transform (Modify) Bone - Bone: dyn_grenade_r_MSTR"(ComponentPose=@n23, Alpha=Get ADS alpha() -> (ADS alpha:double), Translation="0,-4.8,0", Rotation="0,0,0", Scale="1,1,1") {
    translationMode=BMM_Additive
    translationSpace=BCS_BoneSpace
    alphaBoolBlend {blendOption=Linear}
    boneToModify {boneName=dyn_grenade_r_MSTR}
    translation {y=-4.8}
   }
   n25 = "Constraint - dyn_pocketFlap_01_l_MSTR (PPPP)"!(
    ComponentPose=@n24
    Alpha=1
    ConstraintWeights_0=0.139
    ConstraintWeights_1:float=0.478
    ConstraintWeights_2:float=0.112
    ConstraintWeights_3:float=0.271
   ) -> (Pose) {
    alphaBoolBlend {blendOption=Linear}
    boneToModify {boneName=dyn_pocketFlap_01_l_MSTR}
    constraintSetup [{targetBone {boneName=spine_01}}, {targetBone {boneName=spine_02}}, {targetBone {boneName=spine_03}}, {targetBone {boneName=spine_04}}]
    constraintWeights [1,1,1,1]
   }
   n26 = "Constraint - dyn_pocketFlap_02_l_MSTR (PPPP)"!(
    ComponentPose=@n25
    Alpha=1
    ConstraintWeights_0=0.139
    ConstraintWeights_1:float=0.478
    ConstraintWeights_2:float=0.112
    ConstraintWeights_3:float=0.271
   ) -> (Pose) {
    alphaBoolBlend {blendOption=Linear}
    boneToModify {boneName=dyn_pocketFlap_02_l_MSTR}
    constraintSetup [{targetBone {boneName=spine_01}}, {targetBone {boneName=spine_02}}, {targetBone {boneName=spine_03}}, {targetBone {boneName=spine_04}}]
    constraintWeights [1,1,1,1]
   }
   n27 = "Constraint - dyn_pocketFlap_01_r_MSTR (PPPP)"!(
    ComponentPose=@n26
    Alpha=1
    ConstraintWeights_0=0.139
    ConstraintWeights_1:float=0.478
    ConstraintWeights_2:float=0.112
    ConstraintWeights_3:float=0.271
   ) -> (Pose) {
    alphaBoolBlend {blendOption=Linear}
    boneToModify {boneName=dyn_pocketFlap_01_r_MSTR}
    constraintSetup [{targetBone {boneName=spine_01}}, {targetBone {boneName=spine_02}}, {targetBone {boneName=spine_03}}, {targetBone {boneName=spine_04}}]
    constraintWeights [1,1,1,1]
   }
   n28 = "Constraint - dyn_pocketFlap_02_r_MSTR (PPPP)"!(
    ComponentPose=@n27
    Alpha=1
    ConstraintWeights_0=0.139
    ConstraintWeights_1:float=0.478
    ConstraintWeights_2:float=0.112
    ConstraintWeights_3:float=0.271
   ) -> (Pose) {
    alphaBoolBlend {blendOption=Linear}
    boneToModify {boneName=dyn_pocketFlap_02_r_MSTR}
    constraintSetup [{targetBone {boneName=spine_01}}, {targetBone {boneName=spine_02}}, {targetBone {boneName=spine_03}}, {targetBone {boneName=spine_04}}]
    constraintWeights [1,1,1,1]
   }
   n29 = "Constraint - dyn_pistol_l_MSTR (PPP)"!(ComponentPose=@n28, Alpha=1, ConstraintWeights_0=0.04, ConstraintWeights_1:float=0.026, ConstraintWeights_2:float=0.934) -> (Pose) {
    alphaBoolBlend {blendOption=Linear}
    boneToModify {boneName=dyn_pistol_l_MSTR}
    constraintSetup [{targetBone {boneName=pelvis}}, {targetBone {boneName=thigh_l}}, {targetBone {boneName=thigh_twist_01_l}}]
    constraintWeights [1,1,1]
   }
   n30 = "Constraint - dyn_pistol_l_aim (PPP)"!(ComponentPose=@n29, Alpha=1, ConstraintWeights_0=0.01, ConstraintWeights_1:float=0.13, ConstraintWeights_2:float=0.86) -> (Pose) {
    alphaBoolBlend {blendOption=Linear}
    boneToModify {boneName=dyn_pistol_l_aim}
    constraintSetup [{targetBone {boneName=pelvis}}, {targetBone {boneName=thigh_l}}, {targetBone {boneName=thigh_twist_01_l}}]
    constraintWeights [1,1,1]
   }
   n31 = Look At - Bone: dyn_pistol_l_MSTR(ComponentPose=@n30, Alpha=1) {
    bUseLookUpAxis=true
    alphaBoolBlend {blendOption=Linear}
    boneToModify {boneName=dyn_pistol_l_MSTR}
    lookAtLocation {x=0}
    lookAtTarget {boneReference {boneName=dyn_pistol_l_aim}}
    lookAt_Axis {axis=<1 0>}
    lookUp_Axis {axis=<0 1>}
   }
   n32 = "Constraint - dyn_pistol_r_MSTR (PPP)"!(ComponentPose=@n31, Alpha=1, ConstraintWeights_0=0.04, ConstraintWeights_1:float=0.026, ConstraintWeights_2:float=0.934) -> (Pose) {
    alphaBoolBlend {blendOption=Linear}
    boneToModify {boneName=dyn_pistol_r_MSTR}
    constraintSetup [{targetBone {boneName=pelvis}}, {targetBone {boneName=thigh_r}}, {targetBone {boneName=thigh_twist_01_r}}]
    constraintWeights [1,1,1]
   }
   n33 = "Constraint - dyn_pistol_r_aim (PPP)"!(ComponentPose=@n32, Alpha=1, ConstraintWeights_0=0.01, ConstraintWeights_1:float=0.13, ConstraintWeights_2:float=0.86) -> (Pose) {
    alphaBoolBlend {blendOption=Linear}
    boneToModify {boneName=dyn_pistol_r_aim}
    constraintSetup [{targetBone {boneName=pelvis}}, {targetBone {boneName=thigh_r}}, {targetBone {boneName=thigh_twist_01_r}}]
    constraintWeights [1,1,1]
   }
   n34 = Look At - Bone: dyn_pistol_r_MSTR(ComponentPose=@n33, Alpha=1) {
    bUseLookUpAxis=true
    alphaBoolBlend {blendOption=Linear}
    boneToModify {boneName=dyn_pistol_r_MSTR}
    lookAtLocation {x=0}
    lookAtTarget {boneReference {boneName=dyn_pistol_r_aim}}
    lookAt_Axis {axis=<1 0>}
    lookUp_Axis {axis=<0 1>}
   }
   n35 = "Constraint - dyn_belt_strap_r_01_MSTR (PP)"!(ComponentPose=@n34, Alpha=1, ConstraintWeights_0=0.813, ConstraintWeights_1:float=0.187) -> (Pose) {
    alphaBoolBlend {blendOption=Linear}
    boneToModify {boneName=dyn_belt_strap_r_01_MSTR}
    constraintSetup [{targetBone {boneName=spine_05}}, {targetBone {boneName=clavicle_r}}]
    constraintWeights [1,1]
   }
   n36 = "Constraint - dyn_belt_strap_l_01_MSTR (PP)"!(ComponentPose=@n35, Alpha=1, ConstraintWeights_0=0.813, ConstraintWeights_1:float=0.187) -> (Pose) {
    alphaBoolBlend {blendOption=Linear}
    boneToModify {boneName=dyn_belt_strap_l_01_MSTR}
    constraintSetup [{targetBone {boneName=spine_05}}, {targetBone {boneName=clavicle_l}}]
    constraintWeights [1,1]
   }
   n37 = "Constraint - dyn_belt_strap_r_02_MSTR (PPPP)"!(
    ComponentPose=@n36
    Alpha=1
    ConstraintWeights_0=0.262
    ConstraintWeights_1:float=0.175
    ConstraintWeights_2:float=0.456
    ConstraintWeights_3:float=0.107
   ) -> (Pose) {
    alphaBoolBlend {blendOption=Linear}
    boneToModify {boneName=dyn_belt_strap_r_02_MSTR}
    constraintSetup [{targetBone {boneName=pelvis}}, {targetBone {boneName=spine_01}}, {targetBone {boneName=spine_02}}, {targetBone {boneName=spine_03}}]
    constraintWeights [1,1,1,1]
   }
   n38 = "Constraint - dyn_belt_strap_l_02_MSTR (PPPP)"!(
    ComponentPose=@n37
    Alpha=1
    ConstraintWeights_0=0.262
    ConstraintWeights_1:float=0.175
    ConstraintWeights_2:float=0.456
    ConstraintWeights_3:float=0.107
   ) -> (Pose) {
    alphaBoolBlend {blendOption=Linear}
    boneToModify {boneName=dyn_belt_strap_l_02_MSTR}
    constraintSetup [{targetBone {boneName=pelvis}}, {targetBone {boneName=spine_01}}, {targetBone {boneName=spine_02}}, {targetBone {boneName=spine_03}}]
    constraintWeights [1,1,1,1]
   }
   n39 = "Constraint - dyn_hand_strap_r_01_MSTR (PP)"!(ComponentPose=@n38, Alpha=1, ConstraintWeights_0=0.445, ConstraintWeights_1:float=0.555) -> (Pose) {
    alphaBoolBlend {blendOption=Linear}
    boneToModify {boneName=dyn_hand_strap_r_01_MSTR}
    constraintSetup [{targetBone {boneName=hand_r}}, {targetBone {boneName=lowerarm_twist_01_r}}]
    constraintWeights [1,1]
   }
   n40 = "Constraint - dyn_hand_strap_l_01_MSTR (PP)"!(ComponentPose=@n39, Alpha=1, ConstraintWeights_0=0.445, ConstraintWeights_1:float=0.555) -> (Pose) {
    alphaBoolBlend {blendOption=Linear}
    boneToModify {boneName=dyn_hand_strap_l_01_MSTR}
    constraintSetup [{targetBone {boneName=hand_l}}, {targetBone {boneName=lowerarm_twist_01_l}}]
    constraintWeights [1,1]
   }
   n41 = RigidBody(ComponentPose=@n40) {
    bClampLinearTranslationLimitToRefPose=true
    bOverrideWorldGravity=true
    overridePhysicsAsset=&/M_MED_Bullet_Blue_Dyn
    simulationSpace=BaseBoneSpace
    alphaBoolBlend {blendOption=Linear}
    baseBoneRef {boneName=spine_05}
    componentAppliedLinearAccClamp=<3000 3000 3000>
    componentLinearAccScale=<0.4 0.4 0.4>
    componentLinearVelScale=<0.4 0.4 0.4>
    overrideWorldGravity {z=-980}
   }
   n42 = RigidBody(ComponentPose=@n41) {
    bClampLinearTranslationLimitToRefPose=true
    bOverrideWorldGravity=true
    overridePhysicsAsset=&/M_MED_Bullet_Blue_Dyn_legs
    simulationSpace=BaseBoneSpace
    alphaBoolBlend {blendOption=Linear}
    baseBoneRef {boneName=pelvis}
    componentAppliedLinearAccClamp=<3000 3000 3000>
    componentLinearAccScale=<0.35 0.35 0.35>
    componentLinearVelScale=<0.35 0.35 0.35>
    overrideWorldGravity {z=-980}
   }
   Output Pose(Result=Component To Local(ComponentPose=@n42)) {layerGroup=None}
  ]}
 EventGraph {nodes [
   Event BlueprintUpdateAnimation()
   > Set shells fix alpha(shells fix alpha:double="To Float (Boolean)"(InBool=Get bIsParachuteOpened())) -> (Output_Get:double)
   n1 = GetOwnerSkeletalMeshComponent()
   n2 = BreakRotator(InRot=BreakTransform(InTransform=GetDeltaTransformFromRefPose(self=@n1, BoneName=upperarm_r, BaseName=spine_05)).Rotation)
   > Set arm_up_r(arm_up_r:double=MapRangeClamped(Value=@n2.Yaw, InRangeA=110, InRangeB=130, OutRangeA=1, OutRangeB=0)) -> (Output_Get:double)
   > Set arm_forward_l(arm_forward_l:double=MapRangeClamped(Value=@n2.Pitch, InRangeA=-15, InRangeB=-30)) -> (Output_Get:double)
   > Set ADS alpha(ADS alpha:double=float * float(A=Get arm_up_r() -> (arm_up_r:double), B=Get arm_forward_l() -> (arm_forward_l:double))) -> (Output_Get:double)
   TryGetPawnOwner()
   "To Float (Boolean)"(InBool=Get bIsTargeting())
   n3 = BreakTransform(InTransform=GetDeltaTransformFromRefPose(self=GetOwnerSkeletalMeshComponent(), BoneName=upperarm_l, BaseName=spine_05))
   Set arm_up_l(arm_up_l:double=MapRangeClamped(Value=BreakRotator(InRot=@n3.Rotation).Yaw, InRangeA=50, InRangeB=105)) -> (Output_Get:double)
  ]}
]
variables [
 ADS alpha {type=number}
 arm_forward_l {type=number}
 arm_up_l {type=number}
 arm_up_r {type=number}
 shells fix alpha {type=number}
]
