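# Post-copy AnimBlueprint for the face accessory of the male medium BulletBlue cosmetic body, animating its hood.
# Copies the attached parent mesh's pose, then runs RigidBody physics in head bone space with overridden gravity and damped component motion.
# Two additive bone transforms then offset and tilt the left and right hood master bones.
# Each update sets their alpha to zero when bRigidBodyEnabled is true, so the offset applies only without physics.
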AnimBlueprint /BRCosmetics/Characters/Player/Male/Medium/Bodies/M_MED_BulletBlue/Meshes/Parts/M_MED_Bullet_Blue_FaceAcc_AnimBP
target_skeleton=&/M_MED_Bullet_Blue_FaceAcc_Skeleton
graphs [
 AnimGraph {nodes [
   n1 = RigidBody(ComponentPose=Local To Component(LocalPose=Copy Pose From Mesh() {bUseAttachedParent=true})) {
    bClampLinearTranslationLimitToRefPose=true
    bOverrideWorldGravity=true
    overridePhysicsAsset=&/M_MED_Bullet_Blue_FaceAcc_Dyn_hood
    simulationSpace=BaseBoneSpace
    alphaBoolBlend {blendOption=Linear}
    baseBoneRef {boneName=head}
    componentAppliedLinearAccClamp=<3000 3000 3000>
    componentLinearAccScale=<0.25 0.5 0.5>
    componentLinearVelScale=<0.3 0.3 0.3>
    overrideWorldGravity {z=-300}
   }
   n2 = Get modifyAlpha() -> (modifyAlpha:double)
   n3 = "Transform (Modify) Bone - Bone: dyn_hood_r_MSTR"(ComponentPose=@n1, Alpha=@n2, Translation="0.000,0,2.5", Rotation="-5.000,0.000,0.000", Scale="1,1,1") {
    rotationMode=BMM_Additive
    rotationSpace=BCS_BoneSpace
    translationMode=BMM_Additive
    translationSpace=BCS_BoneSpace
    alphaBoolBlend {blendOption=Linear}
    boneToModify {boneName=dyn_hood_r_MSTR}
    rotation {pitch=-5}
    translation {z=2.5}
   }
   n4 = "Transform (Modify) Bone - Bone: dyn_hood_l_MSTR"(ComponentPose=@n3, Alpha=@n2, Translation="0.000,0,2.5", Rotation="-5,0.000,0.000", Scale="1,1,1") {
    rotationMode=BMM_Additive
    rotationSpace=BCS_BoneSpace
    translationMode=BMM_Additive
    translationSpace=BCS_BoneSpace
    alphaBoolBlend {blendOption=Linear}
    boneToModify {boneName=dyn_hood_l_MSTR}
    rotation {pitch=-5}
    translation {z=2.5}
   }
   Output Pose(Result=Component To Local(ComponentPose=@n4)) {layerGroup=None}
  ]}
 EventGraph {nodes [
   Event BlueprintUpdateAnimation()
   > Set modifyAlpha(modifyAlpha:double=SelectFloat(A=0, B=1, bPickA=Get bRigidBodyEnabled())) -> (Output_Get:double)
   TryGetPawnOwner()
  ]}
]
variables [modifyAlpha {type=number}]
